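# Translucent, volumetric per-vertex lit particle material for falling mist, animated with an 8x8 SubUV flipbook on sprites.
# The flipbook's red channel, tinted by particle color and scaled by a dynamic parameter, feeds the emissive adjustment function.
# Its blue channel gets a dynamic contrast power, particle alpha and depth fade for opacity.
# A spherical normal map tiled 8x8 matches the flipbook grid, suggesting rounded shading per frame.
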Material /Game/Effects/Fort_Effects/Materials/Smoke/M_FallingMist_8x8_02
graph {
 properties {
  bEnableResponsiveAA=true
  bUsedWithNanite=true
  bUsedWithNiagaraSprites=true
  bUsedWithParticleSprites=true
  blendMode=BLEND_Translucent
  pixelDepthOffsetMode=PDOM_Legacy
  translucencyLightingMode=TLM_VolumetricPerVertexDirectional
  usedShadingModels=Default Lit
 }
 nodes [
  n1 = ParticleSubUV!(UVs, Tex, Apply View MipBias) {samplerType=SAMPLERTYPE_LinearColor, texture=/Game/Effects/Fort_Effects/Textures/Smoke/T_FallingMist_8x8_01}
  n2 = DynamicParameter() {defaultValue=<rgba 1 0.05 1 1>, paramNames [Alpha Contrast, Emissive Amount, Param3, Param4]}
  n3 = ParticleColor()
  DepthFade(
   Opacity=Multiply(A=Power(Base=@n1.B, Exp=@n2.Alpha Contrast), B=@n3.A)
   FadeDistance=ScalarParameter() {defaultValue=30, parameterName=Depth Fade Distance}
  )
  MaterialFunctionCall(Emissive=Multiply(
    A=Multiply(A=@n1.R, B=@n3.RGB)
    B=DynamicParameter() {defaultValue=<rgba 1 0.05 1 1>, paramNames [Alpha Contrast, Emissive Amount, Param3, Param4]}.Emissive Amount
   )) {materialFunction=&MF_FX_EmissiveAdjustments}
  TextureSample(UVs=TextureCoordinate() {uTiling=8, vTiling=8}) {samplerType=SAMPLERTYPE_Normal, texture=/Game/Gadgets/Assets/Teleporter/Textures/T_Spherical_Normal_02}
 ]
}
