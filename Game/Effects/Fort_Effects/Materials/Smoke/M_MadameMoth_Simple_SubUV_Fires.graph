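# Unlit translucent Niagara sprite material for MadameMoth fire effects, using a fire sprite-sheet texture as an opacity mask.
# Particle color drives emissive through a shared adjustment function.
# Opacity multiplies particle alpha, a selectable mask channel and a tiled smoke-noise lerp, whose minimum can come from a DissolveLerp dynamic parameter.
# Depth fade softens intersections, and a static switch adds squared near-camera fading.
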Material /Game/Effects/Fort_Effects/Materials/Smoke/M_MadameMoth_Simple_SubUV_Fires
graph {
 properties {bUsedWithNiagaraSprites=true, blendMode=BLEND_Translucent, pixelDepthOffsetMode=PDOM_Legacy, shadingModel=MSM_Unlit, usedShadingModels=Unlit}
 nodes [
  n1 = ParticleColor()
  MaterialFunctionCall(Emissive=@n1.RGB) {materialFunction=&MF_FX_EmissiveAdjustments}
  n2 = MaterialFunctionCall(Distance=ScalarParameter() {defaultValue=150, parameterName=Near Camera Fade}) {materialFunction=&Near_Camera_Fade}
  n3 = StaticComponentMaskParameter(None=TextureSampleParameter2D() {parameterName=SubUV_Tex, texture=/Game/Effects/Fort_Effects/Textures/T_MadameMoth_fire_sprites}.RGB) {defaultR=true, parameterName=Mask}
  n4 = Multiply(A=TextureCoordinate() {uTiling=1.5, vTiling=1.5}, B=ScalarParameter() {defaultValue=1, parameterName=LerpTexture_UVTiling})
  n5 = LinearInterpolate(
   A=LinearInterpolate(
    A=ScalarParameter() {parameterName=LerpMin}
    B=DynamicParameter() {defaultValue=<rgba 0 1 1 1>, paramNames [DissolveLerp,Param2,Param3,Param4]}.DissolveLerp
    Alpha=ScalarParameter() {parameterName=Use_DissolveLerp}
   )
   B=ScalarParameter() {defaultValue=1, parameterName=LerpMax}
   Alpha=LinearInterpolate(
    A=TextureSampleParameter2D(UVs=@n4) {parameterName=LerpTexture, texture=/Game/Effects/Fort_Effects/Textures/Tile/T_Inky_Smoke_Tile2}.R
    Alpha=ScalarParameter() {defaultValue=0.1, parameterName=Noise Lerp}
   ) {constAlpha=0.1}
  )
  n6 = DepthFade(Opacity=Multiply(A=@n1.A, B=Multiply(A=@n3, B=@n5)), FadeDistance=ScalarParameter() {defaultValue=20, parameterName=FadeDistance}) {fadeDistanceDefault=5}
  StaticSwitchParameter(True=Multiply(A=Multiply(A=@n2.Result, B=@n2.Result), B=@n6), False=@n6) {parameterName=Use Camera Fade}
 ]
}
