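# Translucent unlit material for distant smoke on meshes, with an optional Niagara particle color path.
# Two panning storm-cloud normal and mask layers drive glow color and opacity, shaped by axis-aligned fresnel, depth fade and camera-distance fade.
# Forward and mobile paths use only the main layer and swap depth fade for a UV fade.
# World position offset adds camera offset, optional drift, and sinks the mesh once fully near-faded.
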Material /Game/Effects/Fort_Effects/Materials/Smoke/M_Mesh_Distant_Smoke
graph {
 properties {
  bAutomaticallySetUsageInEditor=false
  bUsedWithInstancedStaticMeshes=true
  bUsedWithNiagaraMeshParticles=true
  bUsedWithNiagaraSprites=true
  blendMode=BLEND_Translucent
  pixelDepthOffsetMode=PDOM_Legacy
  shadingModel=MSM_Unlit
  usedShadingModels=Unlit
 }
 nodes [
  n1 = TextureCoordinate()
  n2 = AppendVector(
   A=ComponentMask(None=@n1) {r=true}
   B=Multiply(A=ComponentMask(None=@n1) {g=true}, B=ScalarParameter() {defaultValue=1, parameterName=WPOGradient})
  )
  n3 = Time()
  n4 = VectorParameter() {parameterName=WPOMult, defaultValue=<rgba 1 1 1 1>}
  Multiply(
   A=TextureSample(UVs=Panner(Coordinate=@n2, Time=@n3) {speedY=1}) {texture=/Game/Effects/Fort_Effects/Textures/Ramps/T_Power_Source_Sphere_Ramp}.G
   B=@n4.RGB
  )
  n5 = Multiply(A=VertexNormalWS(), B=Multiply(A=VectorNoise() {noiseFunction=VNF_VectorALU}, B=@n4.RGB))
  Multiply(A=@n5, B=@n4.RGB)
  n6 = ParticleColor()
  n7 = Multiply(
   A=ComponentMask(None=TextureCoordinate()) {g=true}
   B=AppendVector(A=Constant(), B=ScalarParameter() {defaultValue=-0.2, parameterName=HeightUVScaleAdd})
  )
  n8 = VectorParameter() {parameterName=TileDetail, defaultValue=<rgba 4 2.5 0 0>}
  n9 = AppendVector(A=@n8.R, B=@n8.R)
  n10 = Panner(
   Coordinate=Multiply(A=TextureCoordinate(), B=Add(A=@n7, B=@n9))
   Speed=ComponentMask(None=VectorParameter() {parameterName=PanSpeedDetail, defaultValue=<rgba -0.015 0.025 0 1>}.RGB) {g=true, r=true}
  ) {bFractionalPart=true}
  n11 = TextureSampleParameter2D(UVs=@n10) {
   parameterName=NormalDetail
   samplerType=SAMPLERTYPE_Normal
   texture=/Game/Characters/Player/Male/Medium/Bodies/M_MED_Hurricane/Textures/T_MED_Hurricane_StormCloud_N
  }
  n12 = VectorParameter() {parameterName=TileMain, defaultValue=<rgba 2 1 0 0>}
  n13 = AppendVector(A=@n12.R, B=@n12.R)
  n14 = Panner(
   Coordinate=Multiply(A=TextureCoordinate(), B=Add(A=@n13, B=@n7))
   Speed=ComponentMask(None=VectorParameter() {parameterName=PanSpeedMain, defaultValue=<rgba -0.01 0.005 0 1>}.RGB) {g=true, r=true}
  ) {bFractionalPart=true}
  n15 = TextureSampleParameter2D(UVs=@n14) {parameterName=NormalMain, samplerType=SAMPLERTYPE_Normal, texture=/Game/Global/Textures/Default/Tileable_Patterns/T_StormCloudDetail01_N}
  n16 = TextureSampleParameter2D(UVs=@n14) {
   parameterName=MaskMain
   samplerType=SAMPLERTYPE_LinearGrayscale
   texture=/Game/Global/Textures/Default/Tileable_Patterns/T_StormCloudDetail01_M
  }
  n17 = Power(Base=@n16.G, Exp=ScalarParameter() {defaultValue=0.6, parameterName=DetailLerpPower})
  n18 = ShadingPathSwitch(Default=LinearInterpolate(A=@n11.RGB, B=@n15.RGB, Alpha=@n17), Forward=@n15.RGB, Mobile=@n15.RGB)
  n19 = TextureSampleParameter2D(UVs=@n10) {
   parameterName=MaskDetail
   samplerType=SAMPLERTYPE_LinearGrayscale
   texture=/Game/Characters/Player/Male/Medium/Bodies/M_MED_Hurricane/Textures/T_MED_Hurricane_StormCloud
  }
  n20 = ShadingPathSwitch(Default=LinearInterpolate(A=@n19.G, B=@n16.G, Alpha=@n17), Forward=@n16.G, Mobile=@n16.G)
  n21 = ScalarParameter() {parameterName=GlowMult}
  n22 = MaterialFunctionCall!(
   GlowDirection
   GlowStartHeight
   GlowWorldPosition
   Normal=Transform(None=@n18)
   Mask=@n20
   Base Glow Mult=@n21
   Inner Glow Mult=@n21
   GlowDirectionBlend
  ) {materialFunction=/Game/Athena/Prototype/Blueprints/White/FX/MF_Giant_Smoke_Pillar_Color}
  Multiply(
   A=ScalarParameter() {defaultValue=1, parameterName=Emissive}
   B=StaticSwitchParameter(True=Multiply(A=@n6.RGB, B=@n22.Result), False=@n22.Result) {parameterName=UseInNiagara?}
  )
  n23 = LinearInterpolate(A=ScalarParameter() {defaultValue=0.5, parameterName=MinOpacity}, Alpha=@n20)
  n24 = ComponentMask(None=TextureCoordinate()) {g=true}
  n25 = MaterialFunctionCall(
   Axis=MaterialFunctionCall() {materialFunction=&TangentBasis}.Y V
   Exponent=LinearInterpolate(
    A=ScalarParameter() {defaultValue=1.2, parameterName=EdgeFadeExpTop, sliderMin=0.5}
    B=ScalarParameter() {defaultValue=0.5, parameterName=EdgeFadeExpBottom}
    Alpha=Multiply(A=@n24, B=@n24)
   ) {constA=2, constB=0.4}
  ) {materialFunction=&AxisAlignedFresnel}
  n26 = Multiply(A=@n25.Fresnel, B=@n25.Fresnel)
  n27 = ScalarParameter() {defaultValue=2, parameterName=PowerTop}
  n28 = ScalarParameter() {defaultValue=0.6, parameterName=PowerBottom}
  n29 = Multiply(A=Power(Base=Multiply(A=@n23, B=@n26), Exp=LinearInterpolate(A=@n27, B=@n28, Alpha=@n20)), B=VertexColor().R)
  n30 = DepthFade(
   Opacity=@n29
   FadeDistance=LinearInterpolate(A=ScalarParameter() {defaultValue=2000, parameterName=DepthFade}, B=ScalarParameter() {defaultValue=150}, Alpha=@n20)
  )
  n31 = OneMinus(None=ComponentMask(None=TextureCoordinate()) {g=true})
  n32 = Multiply(A=@n29, B=Saturate(None=Multiply(A=@n31) {constB=10}))
  n33 = Constant3Vector() {constant=<rgba 1 1 0.15 0>}
  n34 = Distance(A=Multiply(A=ActorPositionWS(), B=@n33), B=Multiply(A=@n33, B=CameraPositionWS()))
  n35 = ScalarParameter() {defaultValue=5000, parameterName=NearFadeStartDistance}
  n36 = Saturate(None=Divide(A=Subtract(A=@n34, B=@n35), B=ScalarParameter() {defaultValue=12000, parameterName=NearFadeDistance}))
  n37 = Multiply(
   A=Multiply(A=ScalarParameter() {defaultValue=1.1, parameterName=Opacity}, B=ShadingPathSwitch(Default=@n30, Forward=@n32, Mobile=@n32)) {constA=1}
   B=@n36
  )
  StaticSwitchParameter(True=Multiply(A=ParticleColor().A, B=@n37), False=@n37) {parameterName=UseInNiagara?}
  n38 = MaterialFunctionCall!(
   WPO Tile=ScalarParameter() {defaultValue=1, parameterName="WPO Tile (S)"}
   WPO Drift Speed=ScalarParameter() {parameterName="WPO Drift Speed (S)"}
   WPO Drift Scale=ScalarParameter() {parameterName="WPO Drift Scale (S)"}
   Texture WPO Scale=ScalarParameter() {parameterName="Texture WPO Scale (S)"}
   Texture WPO Mask=Multiply(
    A=StaticSwitchParameter(True=@n18, False=@n20) {parameterName=UseNormalForWpo?}
    B=ScalarParameter() {defaultValue=1, parameterName=WPO_Test}
   )
  ) {materialFunction=/Game/Athena/Prototype/Blueprints/White/FX/MF_Giant_Smoke_Pillar_WPO}
  n39 = ScalarParameter() {defaultValue=1000, parameterName=CameraOffset}
  n40 = Ceil(None=@n36)
  n41 = Add(
   A=Multiply(
    A=MaterialFunctionCall(Offset Amount=@n39) {materialFunction=/Engine/Functions/Engine_MaterialFunctions02/WorldPositionOffset/CameraOffset}.Camera Offset
    B=Constant3Vector() {constant=<rgba 1 1 0.1 0>}
   )
   B=Multiply(A=OneMinus(None=@n40), B=Constant3Vector() {constant=<rgba 0 0 -50000 0>})
  )
  StaticSwitchParameter(True=Add(A=Add(A=@n38.Drift WPO, B=@n38.Drift WPO), B=@n41), False=@n41) {parameterName=UseExpensiveDetailWPO?}
  StaticSwitchParameter(False=@n5)
 ]
}
