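# Lit translucent material for distant smoke on meshes and Niagara particles, using volumetric per-vertex directional lighting.
# Two panning cloud mask and normal layers blend, with time-of-day normal flattening and giant smoke pillar functions for glow color and WPO.
# Opacity math combines masks, Fresnel edge fade, vertex color, depth fade and camera-proximity fade into temporal dither.
# World-space noise builds normal-direction wobble, while shading path switches simplify noise, blending and fades for Forward and Mobile.
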Material /Game/Effects/Fort_Effects/Materials/Smoke/M_Mesh_Distant_Smoke_Lit
graph {
 properties {
  bUsedWithInstancedStaticMeshes=true
  bUsedWithNiagaraMeshParticles=true
  bUsedWithNiagaraSprites=true
  bUsedWithSplineMeshes=true
  blendMode=BLEND_Translucent
  pixelDepthOffsetMode=PDOM_Legacy
  translucencyDirectionalLightingIntensity=0.6
  translucencyLightingMode=TLM_VolumetricPerVertexDirectional
  usedShadingModels=Default Lit
 }
 nodes [
  n1 = TextureCoordinate()
  n2 = AppendVector(A=ScalarParameter() {defaultValue=2, parameterName=TileX}, B=ScalarParameter() {defaultValue=1.5, parameterName=TileY})
  n3 = Time()
  n4 = ScalarParameter() {defaultValue=1, parameterName=PanSpeed}
  n5 = TextureSampleParameter2D(UVs=Panner(Coordinate=Multiply(A=@n1, B=@n2), Time=Multiply(A=@n3, B=@n4)) {speedX=-0.03, speedY=0.07}) {
   parameterName=WPOMask
   samplerType=SAMPLERTYPE_LinearGrayscale
   texture=/Game/VisualThreatManager/StormVisuals/LargeClouds/T_StormCloudDetail01_M
  }
  n6 = MaterialFunctionCall!(WPO Tile, WPO Drift Speed, WPO Drift Scale, Texture WPO Scale, Texture WPO Mask=@n5.G) {materialFunction=/Game/Athena/Prototype/Blueprints/White/FX/MF_Giant_Smoke_Pillar_WPO}
  Add(A=@n6.Drift WPO, B=@n6.Drift WPO)
  Constant()
  Constant() {r=1}
  n7 = ScalarParameter() {defaultValue=1, parameterName=Opacity}
  n8 = Multiply(
   A=ComponentMask(None=TextureCoordinate()) {g=true}
   B=AppendVector(A=Constant(), B=ScalarParameter() {defaultValue=-0.2, parameterName=HeightUVScaleAdd})
  )
  n9 = VectorParameter() {parameterName=TileDetail, defaultValue=<rgba 4 2.5 0 0>}
  n10 = AppendVector(A=@n9.R, B=@n9.R)
  n11 = Panner(
   Coordinate=Multiply(A=TextureCoordinate(), B=Add(A=@n8, B=@n10))
   Speed=ComponentMask(None=VectorParameter() {parameterName=PanSpeedDetail, defaultValue=<rgba -0.015 0.025 0 1>}.RGB) {g=true, r=true}
  ) {bFractionalPart=true}
  n12 = TextureSampleParameter2D(UVs=@n11) {
   parameterName=MaskDetail
   samplerType=SAMPLERTYPE_LinearGrayscale
   texture=/Game/Characters/Player/Male/Medium/Bodies/M_MED_Hurricane/Textures/T_MED_Hurricane_StormCloud
  }
  n13 = VectorParameter() {parameterName=TileMain, defaultValue=<rgba 2 1 0 0>}
  n14 = AppendVector(A=@n13.R, B=@n13.R)
  n15 = Panner(
   Coordinate=Multiply(A=TextureCoordinate(), B=Add(A=@n14, B=@n8))
   Speed=ComponentMask(None=VectorParameter() {parameterName=PanSpeedMain, defaultValue=<rgba -0.01 0.005 0 1>}.RGB) {g=true, r=true}
  ) {bFractionalPart=true}
  n16 = TextureSampleParameter2D(UVs=@n15) {
   parameterName=MaskMain
   samplerType=SAMPLERTYPE_LinearGrayscale
   texture=/Game/Global/Textures/Default/Tileable_Patterns/T_StormCloudDetail01_M
  }
  n17 = Power(Base=@n16.G, Exp=ScalarParameter() {defaultValue=0.6, parameterName=DetailLerpPower})
  n18 = ShadingPathSwitch(Default=LinearInterpolate(A=@n12.G, B=@n16.G, Alpha=@n17), Forward=@n16.G, Mobile=@n16.G)
  n19 = LinearInterpolate(A=ScalarParameter() {defaultValue=0.5, parameterName=MinOpacity}, Alpha=@n18)
  n20 = ComponentMask(None=TextureCoordinate()) {g=true}
  n21 = MaterialFunctionCall(
   Axis=MaterialFunctionCall() {materialFunction=&TangentBasis}.Y V
   Exponent=LinearInterpolate(
    A=ScalarParameter() {defaultValue=1.2, parameterName=EdgeFadeExpTop, sliderMin=0.5}
    B=ScalarParameter() {defaultValue=0.5, parameterName=EdgeFadeExpBottom}
    Alpha=Multiply(A=@n20, B=@n20)
   ) {constA=2, constB=0.4}
  ) {materialFunction=&AxisAlignedFresnel}
  n22 = Multiply(A=@n21.Fresnel, B=@n21.Fresnel)
  n23 = LinearInterpolate(A=ScalarParameter() {defaultValue=2, parameterName=PowerTop}, B=ScalarParameter() {defaultValue=0.6, parameterName=PowerBottom}, Alpha=@n18)
  n24 = Multiply(A=Power(Base=Multiply(A=@n19, B=@n22), Exp=@n23), B=VertexColor().R)
  n25 = OneMinus(None=ComponentMask(None=TextureCoordinate()) {g=true})
  n26 = Multiply(A=@n24, B=Saturate(None=Multiply(A=@n25) {constB=10}))
  n27 = ShadingPathSwitch(Default=DepthFade(Opacity=@n24, FadeDistance=ScalarParameter() {defaultValue=2000, parameterName=DepthFade}), Forward=@n26, Mobile=@n26)
  n28 = Constant3Vector() {constant=<rgba 1 1 0.15 0>}
  n29 = Distance(A=Multiply(A=ActorPositionWS(), B=@n28), B=Multiply(A=@n28, B=CameraPositionWS()))
  n30 = ScalarParameter() {defaultValue=5000, parameterName=NearFadeStartDistance}
  n31 = Saturate(None=Divide(A=Subtract(A=@n29, B=@n30), B=ScalarParameter() {defaultValue=12000, parameterName=NearFadeDistance}))
  MaterialFunctionCall(Alpha Threshold=Multiply(A=Multiply(A=@n7, B=@n27) {constA=1}, B=@n31)) {materialFunction=&DitherTemporalAA}
  Multiply(A=OneMinus(None=Ceil(None=@n31)), B=Constant3Vector() {constant=<rgba 0 0 -50000 0>})
  n32 = TextureSampleParameter2D(UVs=@n11) {
   parameterName=NormalDetail
   samplerType=SAMPLERTYPE_Normal
   texture=/Game/Characters/Player/Male/Medium/Bodies/M_MED_Hurricane/Textures/T_MED_Hurricane_StormCloud_N
  }
  n33 = TextureSampleParameter2D(UVs=@n15) {parameterName=NormalMain, samplerType=SAMPLERTYPE_Normal, texture=/Game/Global/Textures/Default/Tileable_Patterns/T_StormCloudDetail01_N}
  n34 = ShadingPathSwitch(Default=LinearInterpolate(A=@n32.RGB, B=@n33.RGB, Alpha=@n17), Forward=@n33.RGB, Mobile=@n33.RGB)
  n35 = ScalarParameter() {parameterName=GlowMult}
  Multiply(
   A=ScalarParameter() {parameterName=Emmisive}
   B=MaterialFunctionCall!(
    GlowDirection
    GlowStartHeight
    GlowWorldPosition
    Normal=Transform(None=@n34)
    Mask=@n18
    Base Glow Mult=@n35
    Inner Glow Mult=@n35
    GlowDirectionBlend
   ) {materialFunction=/Game/Athena/Prototype/Blueprints/White/FX/MF_Giant_Smoke_Pillar_Color}.Result
  )
  n36 = ComponentMask(None=TextureCoordinate()) {g=true}
  n37 = Multiply(
   A=Time() {bIgnorePause=true, bOverride_Period=true, period=100}
   B=VectorParameter() {parameterName=PanSpeed, defaultValue=<rgba 250 1500 -500 1>}.RGB
  ) {constB=250}
  n38 = Multiply(
   A=Time() {bIgnorePause=true, bOverride_Period=true, period=500}
   B=VectorParameter() {parameterName=PanSpeedStart, defaultValue=<rgba 250 1500 -500 1>}.RGB
  ) {constB=250}
  n39 = OneMinus(None=ComponentMask(None=TextureCoordinate()) {g=true})
  n40 = Multiply(
   A=LinearInterpolate(A=ScalarParameter() {defaultValue=1, parameterName=s}, B=ScalarParameter() {defaultValue=1, parameterName=s2}, Alpha=OneMinus(None=@n36))
   B=Add(A=LinearInterpolate(A=@n37, B=@n38, Alpha=@n39), B=WorldPosition().XYZ)
  )
  n41 = Noise!(World Position=@n40, FilterWidth) {bTiling=true, bTurbulence=false, levelScale=1, levels=2, noiseFunction=NOISEFUNCTION_ValueALU, repeatSize=500, scale=0.002}
  n42 = ShadingPathSwitch(
   Default=Divide(
    A=Noise(World Position=@n40) {bTiling=true, levelScale=1, levels=4, noiseFunction=NOISEFUNCTION_GradientTex3D, repeatSize=5000, scale=0.0002}
    B=ScalarParameter() {defaultValue=1, parameterName=TestDiv}
   )
   Forward=@n41
   Mobile=@n41
  )
  n43 = ScalarParameter() {defaultValue=50, parameterName=WobbleMult}
  n44 = ScalarParameter() {defaultValue=50, parameterName=WobbleMultStart}
  n45 = ComponentMask(None=TextureCoordinate()) {g=true}
  Multiply(A=VertexNormalWS(), B=Multiply(A=OneMinus(None=@n42), B=LinearInterpolate(A=@n43, B=@n44, Alpha=@n45)) {constB=50})
  n46 = ComponentMask(None=TextureCoordinate()) {g=true, r=true}
  n47 = Multiply(A=ScalarParameter() {defaultValue=800, parameterName=AlphaScale}, B=NamedRerouteUsage())
  n48 = Saturate(None=Divide(A=NamedRerouteUsage()) {constB=50000})
  n49 = Min(A=Multiply(A=@n48, B=@n48)) {constB=0.4}
  n50 = Multiply(A=Multiply(A=Divide(A=@n46, B=@n47) {constB=400}, B=OneMinus(None=@n49))) {constB=2}
  NamedRerouteDeclaration(None=TextureSample(UVs=VertexInterpolator(VS=@n50).PS) {samplerType=SAMPLERTYPE_LinearGrayscale, texture=/Game/Environments/Helios/TechArt/LowAltitudeClouds/T_LowAltitudeClouds_A}.G) {name=ALPHA, nodeColor=<rgba 1 0 0.7734 1>}
  n51 = ScalarParameter() {defaultValue=0.8, parameterName="Normal_Value_Morning (S)"}
  n52 = ScalarParameter() {defaultValue=0.8, parameterName="Normal_Value_Day (S)"}
  n53 = ScalarParameter() {defaultValue=0.8, parameterName="Normal_Value_Evening (S)"}
  n54 = ScalarParameter() {defaultValue=0.8, parameterName="Normal_Value_Night (S)"}
  NamedRerouteDeclaration(None=MaterialFunctionCall(
    Normal=@n34
    Flatness=MaterialFunctionCall(Value_Morning=@n51, Value_Day=@n52, Value_Evening=@n53, Value_Night=@n54) {materialFunction=&Scalar_Time_of_Day}.Result
   ) {materialFunction=&FlattenNormal}.Result) {name=Normal, nodeColor=<rgba 0 1 0.2266 1>}
  NamedRerouteDeclaration(None=Distance(
    A=MaterialFunctionCall() {materialFunction=&ObjectPivotPoint}.Mesh Particle Pivot Location
    B=CameraPositionWS()
   )) {name=Distance Fade, nodeColor=<rgba 0 0.9297 1 1>}
  n55 = MaterialFunctionCall() {materialFunction=&ObjectScale}
  NamedRerouteDeclaration(None=Min(A=Min(A=@n55.Scale X, B=@n55.Scale X), B=@n55.Scale Z)) {name=Min Scale, nodeColor=<rgba 0 0.4375 1 1>}
  NamedRerouteUsage()
  OneMinus(None=@n45)
  OneMinus(None=ComponentMask(None=TextureCoordinate()) {g=true})
  OneMinus(None=MaterialFunctionCall(Axis=Constant3Vector() {constant=<rgba 0 0 1 0>}, Exponent=@n23) {materialFunction=&AxisAlignedFresnel}.Fresnel)
  ScalarParameter() {parameterName=Normal_Flatten}
 ]
}
