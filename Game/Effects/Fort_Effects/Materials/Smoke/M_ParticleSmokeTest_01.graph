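# Translucent lit smoke material for particle sprites, with a fiery emissive tinted by particle color and volumetric per-vertex directional lighting.
# SubUV soft and hard smoke textures feed color, glow and opacity, with dynamic parameters blending texture into base color and soft into hard alpha.
# A panning swirl normal distorts UVs for blended normal and alpha detail.
# Depth fade, varied by the soft texture, softens opacity except at Low quality.
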Material /Game/Effects/Fort_Effects/Materials/Smoke/M_ParticleSmokeTest_01
graph {
 properties {
  bUsedWithParticleSprites=true
  blendMode=BLEND_Translucent
  pixelDepthOffsetMode=PDOM_Legacy
  translucencyLightingMode=TLM_VolumetricPerVertexDirectional
  usedShadingModels=Default Lit
 }
 nodes [
  n1 = ParticleSubUV!(UVs, Tex, Apply View MipBias) {desc=SoftAlpha, samplerType=SAMPLERTYPE_LinearColor, texture=/Game/Effects/Fort_Effects/Textures/Smoke/SmokeTest2_Soft}
  n2 = ParticleSubUV() {desc=HardAlpha, samplerType=SAMPLERTYPE_LinearColor, texture=/Game/Effects/Fort_Effects/Textures/Smoke/SmokeTest2_Hard}
  n3 = LinearInterpolate(A=@n1.R, B=@n2.R)
  n4 = DynamicParameter() {defaultValue=<rgba 0 0 0.2 0>, paramNames [LerpHardSoftAlpha, LerpUVDistortAlpha, PowUVDistortAlpha, NoTexture_Texture]}
  LinearInterpolate(
   A=VectorParameter() {parameterName=BaseColor, defaultValue=<rgba 0.9 0.9 0.9 0>}.RGB
   B=Clamp(None=Add(B=@n3) {constA=0.2})
   Alpha=@n4.NoTexture_Texture
  )
  n5 = Multiply(
   A=ParticleColor().RGB
   B=LinearInterpolate(
    A=VectorParameter() {parameterName=Firey_02}.RGB
    B=VectorParameter() {parameterName=Firey_01, defaultValue=<rgba 8 8 8 0>}.RGB
    Alpha=Multiply(A=@n1.R, B=@n1.R)
   )
  )
  MaterialFunctionCall(Emissive=MaterialFunctionCall(LDR Input=Clamp(None=@n5), HDR Tint=@n5) {materialFunction=&DeriveHDRfromLDR}.RemappedRange) {materialFunction=&MF_FX_EmissiveAdjustments}
  n6 = TextureSample(UVs=Panner(Coordinate=TextureCoordinate() {uTiling=0.5, vTiling=0.5}) {speedY=0.025}) {samplerType=SAMPLERTYPE_Normal, texture=/Game/VisualThreatManager/StormVisuals/LargeClouds/T_StormSwirlNormal}
  n7 = ScalarParameter() {defaultValue=0.012, parameterName=UVDistort}
  n8 = Add(A=TextureCoordinate(), B=ComponentMask(None=Multiply(A=@n6.RGB, B=@n7)) {g=true, r=true})
  n9 = DynamicParameter() {defaultValue=<rgba 0 0 0.2 0>, paramNames [LerpHardSoftAlpha, LerpUVDistortAlpha, PowUVDistortAlpha, NoTexture_Texture]}
  Multiply(
   A=LinearInterpolate(
    A=ParticleSubUV() {samplerType=SAMPLERTYPE_Normal, texture=/Game/Effects/Fort_Effects/Textures/Smoke/SmokeTest_N}.RGB
    B=TextureSample(UVs=@n8) {samplerType=SAMPLERTYPE_Normal, texture=&SmokeTest_N}.RGB
    Alpha=@n9.PowUVDistortAlpha
   )
   B=VectorParameter() {parameterName=ScaleNormal, defaultValue=<rgba 1 1 1.5 0>}.RGB
  )
  n10 = LinearInterpolate(A=@n1.A, B=@n2.A, Alpha=@n9.LerpHardSoftAlpha) {constAlpha=0}
  n11 = TextureSample(UVs=@n8) {desc=UVDistortSoftAlpha, samplerType=SAMPLERTYPE_LinearColor, texture=&SmokeTest2_Soft}
  n12 = Power(Base=@n11.B, Exp=@n9.PowUVDistortAlpha)
  n13 = Multiply(A=LinearInterpolate(A=@n10, B=Multiply(A=@n10, B=@n12), Alpha=@n9.LerpUVDistortAlpha), B=ParticleColor().A)
  QualitySwitch(Default=DepthFade(Opacity=@n13, FadeDistance=LinearInterpolate(Alpha=@n1.G) {constA=2, constB=4}), Low=@n13)
 ]
}
