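# Unlit translucent Niagara sprite material for simple fog or smoke, with optional time-of-day color and opacity.
# An 8x8 smoke flipbook is modulated by panning cloud noise, then added to a sine-shaped radial mask repeated per cell.
# That result drives emissive and opacity with particle color; opacity adds camera fade and depth fade, which mobile skips.
# Dynamic parameters control noise strength, both fades and the time-of-day blend; emissive passes through MF_FX_EmissiveAdjustments.
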Material /Game/Effects/Fort_Effects/Materials/Smoke/M_SimpleFogSmoke_Niagara
graph {
 properties {bUsedWithNiagaraSprites=true, blendMode=BLEND_Translucent, pixelDepthOffsetMode=PDOM_Legacy, shadingModel=MSM_Unlit, usedShadingModels=Unlit}
 nodes [
  n1 = ScalarParameter() {defaultValue=1, group=Time of Day, parameterName=OpacityMorning}
  n2 = ScalarParameter() {defaultValue=1, group=Time of Day, parameterName=OpacityDay}
  n3 = ScalarParameter() {defaultValue=1, group=Time of Day, parameterName=OpacityEvening}
  n4 = ScalarParameter() {defaultValue=1, group=Time of Day, parameterName=OpacityNight}
  n5 = ParticleColor()
  n6 = Frac(None=TextureCoordinate() {uTiling=8, vTiling=8})
  n7 = Constant() {r=0.5}
  n8 = Constant()
  n9 = Multiply(A=Add(A=SphereMask(A=@n6, B=@n7, Radius=@n7, Hardness=@n8)) {constB=1.5}) {constB=0.5}
  n10 = Saturate(None=ConstantBiasScale(None=Sine(None=@n9)))
  n11 = TextureSample(UVs=Panner(Coordinate=TextureCoordinate() {uTiling=8, vTiling=8}) {speedX=0.1, speedY=0.05}) {texture=/Game/Effects/Fort_Effects/Textures/Smoke/T_Clouds_01}
  n12 = DynamicParameter() {defaultValue=<rgba 0.5 1 200 0>, paramNames [Noise Strength, Depth Fade, Camera Fade Length, Use Time of Day]}
  n13 = Add(
   A=Multiply(
    A=ParticleSubUV!(UVs, Tex, Apply View MipBias) {samplerType=SAMPLERTYPE_LinearColor, texture=/Game/Effects/Fort_Effects/Textures/Smoke/T_Smoke_8x8_01}.G
    B=LinearInterpolate(B=@n11.R, Alpha=@n12.Noise Strength) {constA=1}
   )
   B=Multiply(A=Multiply(A=@n10, B=@n10), B=ScalarParameter() {defaultValue=0.5, parameterName=Soft Add})
  )
  n14 = DynamicParameter() {defaultValue=<rgba 0.5 1 200 0>, paramNames [Noise Strength, Depth Fade, Camera Fade Length, Use Time of Day]}
  n15 = DepthFade(Opacity=@n13, FadeDistance=@n14.Depth Fade) {fadeDistanceDefault=30}
  n16 = DynamicParameter() {defaultValue=<rgba 0.5 1 200 0>, paramNames [Noise Strength, Depth Fade, Camera Fade Length, Use Time of Day]}
  n17 = MaterialFunctionCall(Fade Length=@n16.Camera Fade Length) {materialFunction=&CameraDepthFade}
  n18 = Sine(None=Multiply(A=Add(A=@n17.Result) {constB=1.5}) {constB=0.5})
  n19 = Multiply(
   A=Multiply(A=@n5.A, B=ShadingPathSwitch(Default=@n15, Deferred=@n15, Mobile=@n13) {desc=Migrated from FeatureLevelSwitch})
   B=Saturate(None=ConstantBiasScale(None=@n18))
  )
  n20 = DynamicParameter() {defaultValue=<rgba 0.5 1 200 0>, paramNames [Noise Strength, Depth Fade, Camera Fade Length, Use Time of Day]}
  LinearInterpolate(
   A=@n19
   B=Multiply(
    A=@n19
    B=MaterialFunctionCall(Color_Morning=@n1, Color_Day=@n2, Color_Evening=@n3, Color_Night=@n4) {materialFunction=&Color_Time_of_Day}.Result
   )
   Alpha=@n20.Use Time of Day
  )
  n21 = MaterialFunctionCall(
   Color_Morning=VectorParameter() {parameterName=Color Morning, defaultValue=<rgba 0.5156 0.2559 0.1782 0>}.RGB
   Color_Day=VectorParameter() {parameterName=Color Day, defaultValue=<rgba 1 1 1 0>}.RGB
   Color_Evening=VectorParameter() {parameterName=Color Evening, defaultValue=<rgba 0.5156 0.2559 0.1782 0>}.RGB
   Color_Night=VectorParameter() {parameterName=Color Night, defaultValue=<rgba 0.1406 0.1853 0.349 0>}.RGB
  ) {materialFunction=&Color_Time_of_Day}
  n22 = Multiply(A=@n5.RGB, B=@n13)
  MaterialFunctionCall(Emissive=LinearInterpolate(A=@n22, B=Multiply(A=@n21.Result, B=@n22), Alpha=@n20.Use Time of Day)) {materialFunction=&MF_FX_EmissiveAdjustments}
 ]
}
