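# Unlit, translucent, two-sided smoke material for Niagara sprites and instanced meshes, with quality switches that skip depth fade on mobile and lower settings.
# An 8x8 smoke flipbook with motion-vector blending supplies color and opacity, animated by time and dynamic-parameter speed; a plain SubUV sample is also declared.
# Emissive blends particle-tinted texture with a low-to-high color ramp and optional time-of-day tint.
# Dynamic parameters drive opacity erosion, depth fade and switchable near-camera fade.
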Material /Game/Effects/Fort_Effects/Materials/Smoke/M_Smoke_Detailed_Cheap
graph {
 properties {
  bUsedWithInstancedStaticMeshes=true
  bUsedWithNiagaraSprites=true
  blendMode=BLEND_Translucent
  pixelDepthOffsetMode=PDOM_Legacy
  shadingModel=MSM_Unlit
  twoSided=true
  usedShadingModels=Unlit
 }
 nodes [
  n1 = NamedRerouteUsage()
  n2 = ShadingPathSwitch(Default=ShadingPathSwitch(Default=QualitySwitch(Default=@n1, Low=@n1), Mobile=@n1) {desc=Migrated from FeatureLevelSwitch}, Forward=@n1)
  n3 = ParticleColor()
  n4 = LinearInterpolate(
   A=Constant() {r=1}
   B=MaterialFunctionCall(
    Color_Morning=VectorParameter() {group=ToD, parameterName=TOD Morning, defaultValue=<rgba 1 0.3343 0.122 0>}.RGB
    Color_Day=VectorParameter() {group=ToD, parameterName=TOD Day, defaultValue=<rgba 1 1 1 0>}.RGB
    Color_Evening=VectorParameter() {group=ToD, parameterName=TOD Evening, defaultValue=<rgba 1 0.1673 0.1291 0>}.RGB
    Color_Night=VectorParameter() {group=ToD, parameterName=TOD Night, defaultValue=<rgba 0.229 0.442 1 0>}.RGB
   ) {materialFunction=&Color_Time_of_Day}.Result
   Alpha=ScalarParameter() {group=ToD, parameterName=Use ToD Color}
  ) {constA=1}
  n5 = Multiply(
   A=LinearInterpolate(
    A=VectorParameter() {group=Color, parameterName=Color Low, defaultValue=<rgba 0.0156 0.0117 0.0084 1>}.RGB
    B=VectorParameter() {group=Color, parameterName=Color High, defaultValue=<rgba 1 1 1 0>}.RGB
    Alpha=@n2
   )
   B=@n3.RGB
  )
  MaterialFunctionCall(Emissive=LinearInterpolate(A=Multiply(A=@n2, B=@n3.RGB), B=Multiply(A=@n4, B=@n5), Alpha=ScalarParameter() {group=Color, parameterName=Use Color Lerp})) {materialFunction=&MF_FX_EmissiveAdjustments}
  n6 = Time()
  n7 = DynamicParameter() {defaultValue=<rgba 0 1 1 1>, paramNames [Erosion, Erosion_Softness, AnimationSpeed, Param4]}
  n8 = MaterialFunctionCall!(
   "Animation  Phase (0-1)"=Frac(None=Multiply(A=@n6, B=@n7.AnimationSpeed) {constB=0.1})
   Number of Columns X=Constant() {r=8}
   Number of Rows Y=Constant() {r=8}
   Texture=TextureObject() {texture=/Game/Effects/Fort_Effects/Textures/Smoke/T_Smoke_Thick_8x8_02}
   UVs
   Motion Vector Texture=TextureObject()
   Phase to Debug
   Motion Vector Intensity=Multiply(A=@n7.AnimationSpeed, B=ScalarParameter() {defaultValue=-0.1, group=1. TextureValues, parameterName="Motion Vector Intensity (S)"})
   Double Apply Motion Vectors
   Custom UV Input
   Custom UV2 Input
   Use Custom UVs
  ) {materialFunction=/Engine/Functions/Engine_MaterialFunctions02/Texturing/FlipBook_MotionVectors}
  NamedRerouteDeclaration(None=@n8.Result) {name=EmissiveTexture, nodeColor=<rgba 0 1 0.2734 1>}
  NamedRerouteDeclaration(None=@n8.Alpha) {name=OpacityTexture, nodeColor=<rgba 1 0 0.4219 1>}
  n9 = ParticleSubUV!(UVs, Tex, Apply View MipBias) {texture=&T_Smoke_Thick_8x8_02}
  NamedRerouteDeclaration(None=@n9.RGB) {name=EmissiveTexture_Low, nodeColor=<rgba 0 1 0.2734 1>}
  NamedRerouteDeclaration(None=@n9.A) {name=OpacityTexture_Low, nodeColor=<rgba 1 0 0.4219 1>}
  NamedRerouteUsage()
  NamedRerouteUsage()
  n10 = DynamicParameter() {parameterIndex=1, defaultValue=<rgba 250 100 1 1>, paramNames [DepthFade,NearFade,---,---]}
  n11 = NamedRerouteUsage()
  n12 = ShadingPathSwitch(Default=ShadingPathSwitch(Default=QualitySwitch(Default=@n11, Low=@n11), Mobile=@n11) {desc=Migrated from FeatureLevelSwitch}, Forward=@n11)
  n13 = Multiply(
   A=@n12
   B=DepthFade(FadeDistance=DynamicParameter() {parameterIndex=1, defaultValue=<rgba 250 100 1 1>, paramNames [DepthFade,NearFade,---,---]}.DepthFade)
  )
  n14 = Subtract(
   A=ShadingPathSwitch(Default=QualitySwitch(Default=@n13, Low=@n12, Medium=@n12), Mobile=@n12) {desc=Migrated from FeatureLevelSwitch}
   B=DynamicParameter() {defaultValue=<rgba 0 1 1 1>, paramNames [Erosion, Erosion_Softness, AnimationSpeed, Param4]}.Erosion
  )
  n15 = DynamicParameter() {defaultValue=<rgba 0 1 1 1>, paramNames [Erosion, Erosion_Softness, AnimationSpeed, Param4]}
  n16 = Multiply(A=Saturate(None=Divide(A=@n14, B=@n15.Erosion_Softness)), B=ParticleColor().A)
  StaticSwitchParameter(
   True=Multiply(
    A=MaterialFunctionCall(Distance=@n10.NearFade) {materialFunction=&Near_Camera_Fade}.Result
    B=@n16
   )
   False=@n16
  ) {parameterName=UseNearFade?}
 ]
}
