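# Unlit, translucent, two-sided smoke particle material that samples an 8x8 flipbook, with configurable colors.
# The texture's green channel blends Color Low to Color High, multiplied by particle color and an optional time-of-day tint, then passed through shared emissive adjustments.
# Opacity uses texture alpha minus dynamic-parameter erosion weighted by inverse green, times particle alpha.
# Dynamic-parameter depth fade is skipped on mobile and Low/Medium quality.
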Material /Game/Effects/Fort_Effects/Materials/Smoke/M_Smoke_Detailed_ColorOptions
graph {
 properties {
  bUsedWithInstancedStaticMeshes=true
  bUsedWithNiagaraSprites=true
  blendMode=BLEND_Translucent
  pixelDepthOffsetMode=PDOM_Legacy
  shadingModel=MSM_Unlit
  twoSided=true
  usedShadingModels=Unlit
 }
 nodes [
  n1 = MaterialFunctionCall(
   Color_Morning=VectorParameter() {group=ToD, parameterName=TOD Morning, defaultValue=<rgba 1 0.3343 0.122 0>}.RGB
   Color_Day=VectorParameter() {group=ToD, parameterName=TOD Day, defaultValue=<rgba 1 1 1 0>}.RGB
   Color_Evening=VectorParameter() {group=ToD, parameterName=TOD Evening, defaultValue=<rgba 1 0.1673 0.1291 0>}.RGB
   Color_Night=VectorParameter() {group=ToD, parameterName=TOD Night, defaultValue=<rgba 0.229 0.442 1 0>}.RGB
  ) {materialFunction=&Color_Time_of_Day}
  n2 = Constant() {r=1}
  n3 = ParticleColor()
  n4 = ParticleSubUV!(UVs, Tex, Apply View MipBias) {texture=/Game/Effects/Fort_Effects/Textures/Smoke/T_Smoke_Thick_8x8_02}
  n5 = LinearInterpolate(
   A=VectorParameter() {parameterName=Color Low, defaultValue=<rgba 0.0156 0.0117 0.0084 1>}.RGB
   B=VectorParameter() {parameterName=Color High, defaultValue=<rgba 1 1 1 0>}.RGB
   Alpha=@n4.G
  )
  MaterialFunctionCall(Emissive=Multiply(A=StaticSwitchParameter(True=@n1.Result, False=@n2) {parameterName=Use Time of Day colors}, B=Multiply(A=@n5, B=@n3.RGB))) {materialFunction=&MF_FX_EmissiveAdjustments}
  n6 = Multiply(A=OneMinus(None=@n4.G), B=DynamicParameter() {defaultValue=<rgba 25 0 0 0>, paramNames [DepthFade,Erosion,---,---]}.Erosion)
  n7 = Multiply(A=@n4.A, B=DepthFade(FadeDistance=DynamicParameter() {defaultValue=<rgba 25 0 0 0>, paramNames [DepthFade,Erosion,---,---]}.DepthFade))
  n8 = QualitySwitch(Default=@n7, Low=@n4.A, High=@n7, Medium=@n4.A, Epic=@n7)
  n9 = ShadingPathSwitch(Default=@n8, Deferred=@n8, Mobile=@n4.A) {desc=Migrated from FeatureLevelSwitch}
  Multiply(A=Saturate(None=Subtract(A=@n9, B=@n6)), B=ParticleColor().A)
 ]
}
