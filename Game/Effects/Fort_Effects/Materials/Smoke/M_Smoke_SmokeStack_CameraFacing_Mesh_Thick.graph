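# Unlit translucent two-sided material for thick smoke-stack plume meshes that turn around their axis to face the camera.
# World position offset combines camera yaw, height-based bend and panning normal-map waves, with waves replaced by a constant on low-end paths.
# Two panning textures, offset by view angle, distort soft opacity masks with depth and distance fades.
# Emissive color blends time-of-day tint with directional normal shading and optional Schlick-phase sun scattering.
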Material /Game/Effects/Fort_Effects/Materials/Smoke/M_Smoke_SmokeStack_CameraFacing_Mesh_Thick
graph {
 properties {
  bTangentSpaceNormal=false
  bUseTranslucencyVertexFog=false
  blendMode=BLEND_Translucent
  pixelDepthOffsetMode=PDOM_Legacy
  shadingModel=MSM_Unlit
  twoSided=true
  usedShadingModels=Unlit
 }
 nodes [
  n1 = Constant() {r=0.25}
  n2 = CameraPositionWS()
  n3 = Constant3Vector() {constant=<rgba 1 1 0 0>}
  MaterialExpressionNormalize_1 = Normalize(VectorInput=Subtract(A=ObjectPositionWS(), B=Multiply(A=@n2, B=@n3)))
  n4 = Custom!(In=@MaterialExpressionNormalize_1) {
   code="float2 output; \r\nfloat2 input; \r\noutput= atan2 (In.y,In.x);\r\nreturn (output);"
   outputType=CMOT_Float2
   inputs [{
     inputName=In
     input {expression=&self:MaterialExpressionNormalize_1}
    }]
  }
  n5 = Divide(A=Divide(A=@n4) {constB=3.142}, B=Constant() {r=2})
  n6 = TextureCoordinate()
  n7 = AppendVector(A=@n6, B=ComponentMask(None=@n6) {r=true})
  n8 = ComponentMask(None=TextureCoordinate()) {g=true}
  n9 = Multiply(
   A=Power(Base=OneMinus(None=@n8), Exp=ScalarParameter() {defaultValue=2, parameterName=BendExp})
   B=ScalarParameter() {defaultValue=2, parameterName=BendAmount}
  )
  n10 = VectorParameter() {parameterName=BendDirection, defaultValue=<rgba 1 0 0 1>}
  n11 = Add(
   A=RotateAboutAxis(
    NormalizedRotationAxis=ObjectOrientation()
    RotationAngle=Add(A=@n1, B=@n5)
    PivotPoint=TransformPosition(None=@n7) {transformType=TRANSFORMPOSSOURCE_World}
    Position=WorldPosition().XYZ
   )
   B=Multiply(
    A=Multiply(A=@n9, B=@n10.RGB)
    B=MaterialFunctionCall() {materialFunction=&ObjectScale}.Scale Z
   )
  )
  n12 = Constant3Vector()
  n13 = TextureCoordinate()
  n14 = ComponentMask(None=VectorParameter() {group=WPO_Wave, parameterName=TextureWPO_UVtiling, defaultValue=<rgba 0.1 0.1 0 0>}.RGB) {g=true, r=true}
  n15 = VectorParameter() {group=WPO_Wave, parameterName=TextureWPO_Speed, defaultValue=<rgba 0.2 0.2 0 0>}
  n16 = TextureSample(UVs=Panner(Coordinate=Multiply(A=@n13, B=@n14), Speed=ComponentMask(None=@n15.RGB) {g=true, r=true}) {speedY=0.15}) {samplerType=SAMPLERTYPE_Normal, texture=/Game/Global/Textures/Default/Tileable_Patterns/2ndInfectedMapNormal}
  n17 = ComponentMask(None=@n16.RGB) {r=true}
  n18 = ScalarParameter() {defaultValue=1, group=WPO_Wave, parameterName=WPO_HeightX}
  n19 = ComponentMask(None=@n16.RGB) {g=true}
  n20 = ScalarParameter() {group=WPO_Wave, parameterName=WPO_HeightY}
  n21 = AppendVector(
   A=AppendVector(A=Multiply(A=@n17, B=@n18) {constB=0.5}, B=Multiply(A=@n19, B=@n20) {constB=0.5})
   B=Multiply(A=@n16.B, B=ScalarParameter() {group=WPO_Wave, parameterName=WPO_HeightZ}) {constB=0.6}
  )
  n22 = Constant() {r=0.5}
  n23 = SphereMask(
   A=MaterialFunctionCall() {materialFunction=&LinearGradient}.VGradient
   B=@n22
   Radius=@n22
  ) {hardnessPercent=0}
  n24 = Multiply(
   A=Transform(None=LinearInterpolate(A=@n12, B=@n21, Alpha=@n23))
   B=LinearInterpolate(
    A=ScalarParameter() {group=WPO_Wave, parameterName=WPO_Wave_Intensity_Close}
    B=ScalarParameter() {defaultValue=1, group=WPO_Wave, parameterName=WPO_Wave_Intenstiy}
   )
  ) {constB=80}
  n25 = Constant3Vector() {desc=Disable WPO anim in Low, constant=<rgba 0 0 20 0>}
  n26 = FeatureLevelSwitch(Default=@n24, ES2=@n25, ES3_1=@n25, SM4_REMOVED=@n24, SM5=@n24, SM6=@n24)
  Add(A=Multiply(A=@n11), B=QualitySwitch(Default=@n26, Low=@n25, High=@n26, Medium=@n26))
  ComponentMask() {g=true}
  n27 = ShadingPathSwitch(
   Default=ScalarParameter() {defaultValue=1, group=CloseFarFader, parameterName=DistanceScaleMult}
   Deferred=QualitySwitch(Default=Constant() {r=1}, Low=ScalarParameter() {defaultValue=1, group=CloseFarFader, parameterName=DistanceScaleMult_LowQuality})
  ) {desc=Squish WPO earlier on lower quality}
  MaterialFunctionCall!(
   NearFadeLength=Divide(A=ScalarParameter() {defaultValue=21000, group=CloseFarFader, parameterName="NearFadeLength (S)"}, B=@n27)
   NearFadeOffset=Divide(A=ScalarParameter() {defaultValue=20000, group=CloseFarFader, parameterName="NearFadeOffset (S)"}, B=@n27)
   NearFadeCurve
   NearFadeStartingScale
   NearFadeEndingScale
   FarFadeLength=Multiply(A=ScalarParameter() {defaultValue=25000, group=CloseFarFader, parameterName="FarFadeLength (S)"}, B=@n27)
   FarFadeOffset=Multiply(A=ScalarParameter() {defaultValue=25000, group=CloseFarFader, parameterName="FarFadeOffset (S)"}, B=@n27)
   FarFadeCurve
   FarFadeEndingScale
   Opacity Near Distance Bias
   Opacity Far Distance Bias=ScalarParameter() {group=CloseFarFader, parameterName="Opacity Far Distance Bias (S)"}
  ) {materialFunction=/Game/Effects/Fort_Effects/Materials/_Functions/MF_SmoothLod}
  n28 = ScalarParameter() {defaultValue=1, parameterName=OpacityMult}
  n29 = TextureCoordinate()
  n30 = ComponentMask(None=@n29) {r=true}
  n31 = Constant() {r=0.5}
  n32 = Constant() {r=0.5}
  n33 = Constant()
  n34 = Multiply(A=Power(Base=SphereMask(A=@n30, B=@n31, Radius=@n32, Hardness=@n33)) {constExponent=1.25}) {constB=10}
  n35 = Saturate(None=Multiply(A=ComponentMask(None=@n29) {g=true}) {constB=5})
  n36 = ObjectPositionWS()
  n37 = Constant3Vector() {constant=<rgba 1 1 0 0>}
  n38 = CameraPositionWS()
  n39 = Constant3Vector() {constant=<rgba 1 1 0 0>}
  n40 = Normalize(VectorInput=Subtract(A=Multiply(A=@n36, B=@n37), B=Multiply(A=@n38, B=@n39)))
  n41 = ComponentMask(None=@n40) {r=true}
  n42 = ComponentMask(None=@n40) {g=true}
  n43 = Add(A=Divide(A=Arctangent2!(Y=@n41, X=@n42)) {constB=3.142})
  n44 = VertexInterpolator(VS=AppendVector(A=Multiply(A=@n43) {constB=0.5}, B=Constant()))
  n45 = TextureCoordinate()
  n46 = ComponentMask(None=@n45) {g=true}
  n47 = ScalarParameter() {defaultValue=2, parameterName=YScale}
  n48 = AppendVector(
   A=Multiply(A=ComponentMask(None=@n45) {r=true}, B=ScalarParameter() {defaultValue=0.25, parameterName=XScale})
   B=Power(Base=Multiply(A=@n46, B=@n47), Exp=ScalarParameter() {defaultValue=2, parameterName=YScale_LengthTiler})
  )
  n49 = ScalarParameter() {parameterName=XSpeed}
  n50 = ScalarParameter() {defaultValue=0.25, parameterName=YSpeed}
  n51 = TextureSampleParameter2D(UVs=Panner(Coordinate=Add(A=@n44.PS, B=@n48), Speed=AppendVector(A=@n49, B=@n50))) {parameterName=Texture 01, samplerType=SAMPLERTYPE_Normal, texture=&2ndInfectedMapNormal}
  n52 = TextureCoordinate()
  n53 = ComponentMask(None=@n52) {g=true}
  n54 = ScalarParameter() {defaultValue=2, parameterName=YScale2}
  n55 = AppendVector(
   A=Multiply(A=ComponentMask(None=@n52) {r=true}, B=ScalarParameter() {defaultValue=0.25, parameterName=XScale2})
   B=Power(Base=Multiply(A=@n53, B=@n54), Exp=ScalarParameter() {defaultValue=2, parameterName=YScale_LengthTiler_02})
  )
  n56 = ScalarParameter() {parameterName=XSpeed_02}
  n57 = ScalarParameter() {defaultValue=0.25, parameterName=YSpeed_02}
  n58 = TextureSampleParameter2D(UVs=Panner(Coordinate=Add(A=@n44.PS, B=@n55), Speed=AppendVector(A=@n56, B=@n57))) {parameterName=Texture 02, samplerType=SAMPLERTYPE_Normal, texture=&2ndInfectedMapNormal}
  n59 = Add(A=@n51.RGB, B=@n58.RGB)
  n60 = ComponentMask(None=TextureCoordinate() {vTiling=0.7}) {g=true}
  n61 = SphereMask(
   A=LinearInterpolate(
    A=TextureCoordinate() {uTiling=2}
    B=ComponentMask(None=@n59) {g=true, r=true}
    Alpha=LinearInterpolate(Alpha=@n60) {constA=0.5, constAlpha=0, constB=0}
   )
   B=Constant2Vector() {g=0.55, r=0.8}
   Radius=Constant() {r=0.4}
   Hardness=Constant() {r=0.25}
  )
  n62 = SmoothStep(
   Min=ScalarParameter() {parameterName=SmoothStep_Min}
   Max=ScalarParameter() {defaultValue=1, parameterName=SmoothStep_Max}
   Value=Multiply(A=Saturate(None=@n34), B=Multiply(A=@n35, B=@n61))
  )
  n63 = ScalarParameter() {defaultValue=30000, parameterName="Fade_Start_Distance (S)"}
  n64 = ScalarParameter() {defaultValue=25000, parameterName="Fade_End_Distance (S)"}
  Multiply(
   A=DepthFade(Opacity=Multiply(A=@n62, B=@n28), FadeDistance=ScalarParameter() {parameterName=FadeDistance})
   B=VertexInterpolator(VS=MaterialFunctionCall(Fade_Start_Distance=@n63, Fade_End_Distance=@n64) {materialFunction=&MF_Distance_SyncedFade_FX}.Result).PS
  )
  Multiply()
  NamedRerouteDeclaration(None=@n11) {name=WPO, nodeColor=<rgba 0 1 0.4844 1>}
  Normalize(VectorInput=Subtract(
    A=MaterialFunctionCall() {materialFunction=&CameraDirectionVector}.Result
    B=AtmosphericLightVector()
   ))
  n65 = CollectionParameter() {
   collection=&FortniteMaterialParameters
   parameterName=SunlightDirectionalVector
  }
  MaterialExpressionDotProduct_1 = DotProduct(A=ComponentMask(None=@n65) {b=true, g=true, r=true}, B=CameraVectorWS())
  MaterialExpressionScalarParameter_22 = ScalarParameter() {defaultValue=0.4, parameterName=Anisotropy}
  n66 = Custom!(CosTheta=@MaterialExpressionDotProduct_1, k=@MaterialExpressionScalarParameter_22) {
   code="\tfloat Inner = (1 + k * CosTheta);\r\n\treturn (1 - k * k) / (4 * PI * Inner * Inner);\r\n\r\n"
   description=Schlick Phase
   inputs [
    {
     inputName=CosTheta
     input {expression=&self:MaterialExpressionDotProduct_1}
    }
    {
     inputName=k
     input {expression=&self:MaterialExpressionScalarParameter_22}
    }
   ]
  }
  n67 = ScalarParameter() {defaultValue=28, parameterName=ScaleCloudColor}
  n68 = VectorParameter() {parameterName=BaseSmokeColor, defaultValue=<rgba 1 1 1 1>}
  n69 = CollectionParameter() {
   collection=&FortniteMaterialParameters
   parameterName=SunAndMoonModelDirectionalVector
  }
  n70 = Multiply(A=ComponentMask(None=@n69) {b=true, g=true, r=true}) {constB=-1}
  n71 = MaterialFunctionCall!(Value=@n62) {materialFunction=/Engine/Functions/Engine_MaterialFunctions03/Procedurals/PreparePerturbNormalHQ}
  n72 = VertexNormalWS()
  n73 = MaterialFunctionCall!(
   Bump one pixel right=@n71.Value one pixel right
   Bump center=@n71.Value one pixel right
   Bump one pixel down=@n71.Value one pixel right
   WS Normal=Normalize(VectorInput=Subtract(A=@n72, B=@n11))
  ) {materialFunction=/Engine/Functions/Engine_MaterialFunctions03/Procedurals/PerturbNormalHQ}
  n74 = AppendVector(
   A=Multiply(A=ComponentMask(None=@n73.WS Normal) {g=true, r=true}, B=ScalarParameter() {defaultValue=1, parameterName=NormalIntensity})
   B=ComponentMask(None=@n73.WS Normal) {b=true}
  )
  n75 = LinearInterpolate(Alpha=Saturate(None=DotProduct(A=@n70, B=@n74)))
  n76 = LinearInterpolate(A=@n68.RGB, B=Multiply(A=@n68.RGB) {constB=0.25}, Alpha=@n75)
  n77 = MaterialFunctionCall() {materialFunction=/Game/Effects/Fort_Effects/Materials/_Functions/CurveColor_Time_of_Day}
  n78 = LinearInterpolate(A=@n77.Result, B=Multiply(A=@n77.Result) {constB=0.5}, Alpha=Saturate(None=@n62))
  n79 = LinearInterpolate(A=@n78, B=Multiply(A=@n78, B=ScalarParameter() {defaultValue=0.5, parameterName=DarkColor}) {constB=0.25}, Alpha=@n75)
  n80 = ScalarParameter() {defaultValue=1, parameterName=TOD_Amount}
  n81 = Multiply(
   A=ScalarParameter() {defaultValue=0.25, parameterName=EmissiveColor_Mult}
   B=StaticSwitchParameter(True=LinearInterpolate(A=@n76, B=@n79, Alpha=@n80), False=@n68.RGB) {defaultValue=true, parameterName=UseTimeofDayColor?}
  )
  StaticSwitchParameter(True=Multiply(A=Multiply(A=@n66, B=@n67), B=@n81), False=@n81) {parameterName=UseLightThroughClouds?}
 ]
}
