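# Translucent, two-sided, lit smoke trail material for campfire effects, usable on particle sprites and beam trails.
# Opacity multiplies particle alpha, a parabolic fade across UV0's U axis, and a soft smoke texture slowly panned on UV1.
# Near-camera, view-angle and depth fades further soften it.
# Particle blue blends in an orange emissive color through an FX emissive adjustment function, alongside a white SmokeColor parameter.
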Material /Game/Effects/Fort_Effects/Materials/Smoke/M_Smoke_Trail_Campfire
graph {
 properties {
  bUsedWithBeamTrails=true
  bUsedWithParticleSprites=true
  blendMode=BLEND_Translucent
  pixelDepthOffsetMode=PDOM_Legacy
  translucencyLightingMode=TLM_VolumetricPerVertexDirectional
  twoSided=true
  usedShadingModels=Default Lit
 }
 nodes [
  n1 = ParticleColor()
  n2 = ComponentMask(None=TextureCoordinate()) {r=true}
  n3 = LinearInterpolate(Alpha=Multiply(A=OneMinus(None=@n2), B=@n2)) {constB=4}
  n4 = Panner(Coordinate=TextureCoordinate() {coordinateIndex=1, uTiling=2}) {speedX=0.025}
  n5 = Multiply(
   A=@n1.A
   B=Multiply(
    A=Clamp(None=@n3)
    B=TextureSample(UVs=@n4) {samplerType=SAMPLERTYPE_Grayscale, texture=&T_Smoke_Trail_Soft}.R
   )
  )
  n6 = MaterialFunctionCall(Distance=ScalarParameter() {defaultValue=500, parameterName=NearCameraFadeDistance}) {materialFunction=&Near_Camera_Fade}
  n7 = ScalarParameter() {defaultValue=1.5, parameterName=Fade Sensitivity}
  DepthFade(
   Opacity=Multiply(
    A=Multiply(A=@n5, B=@n6.Result)
    B=MaterialFunctionCall("FadeSensitivity (optional)"=@n7) {materialFunction=&FadeBasedOnViewAngle}.Result
   )
   FadeDistance=ScalarParameter() {defaultValue=65, parameterName=Fade Distance}
  )
  MaterialFunctionCall(Emissive=LinearInterpolate(B=VectorParameter() {parameterName=EmissiveColor, defaultValue=<rgba 1 0.4614 0.0871 0>}.RGB, Alpha=@n1.B)) {materialFunction=&MF_FX_EmissiveAdjustments}
  VectorParameter() {parameterName=SmokeColor, defaultValue=<rgba 1 1 1 0>}
 ]
}
