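# Translucent unlit particle material for swirling smoke sprites, animated from an 8x8 SubUV texture and tinted by particle color.
# Opacity combines dynamic-parameter texture shaping and per-cell gradient fade with depth fade, optional near-camera fade and optional Apollo snow-progression masking.
# Emissive fakes lighting by blending dark tint to particle color using a flipbook normal dotted with sun/moon or fake light direction.
# Near-camera shrink pulls vertices toward particle centers; mobile gets cheaper fallbacks.
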Material /Game/Effects/Fort_Effects/Materials/Smoke/M_SwirlSmoke_Unlit
graph {
 properties {
  bAllowNegativeEmissiveColor=true
  bUsedWithNiagaraSprites=true
  bUsedWithParticleSprites=true
  blendMode=BLEND_Translucent
  pixelDepthOffsetMode=PDOM_Legacy
  shadingModel=MSM_Unlit
  translucencyLightingMode=TLM_VolumetricPerVertexDirectional
  usedShadingModels=Unlit
 }
 nodes [
  n1 = ParticleColor()
  n2 = ParticleSubUV!(UVs, Tex, Apply View MipBias) {texture=/Game/Effects/Fort_Effects/Textures/Smoke/T_SmokeSwirl_SubUV_Simple_8x8}
  n3 = DynamicParameter() {defaultValue=<rgba 1 1 0 1>, paramNames [Texture_Exponent, Texture_Hardness, DirectionalFade, Not Used]}
  n4 = Clamp(None=Power(Base=Multiply(A=@n2.G, B=@n3.Texture_Hardness), Exp=@n3.Texture_Exponent))
  n5 = ScalarParameter() {parameterName=UseTexture-DynamicParam}
  n6 = Multiply(A=@n1.A, B=Multiply(A=LinearInterpolate(A=@n2.G, B=@n4, Alpha=@n5)) {constB=1.5})
  n7 = ScalarParameter() {defaultValue=35, parameterName=DepthFadeDist}
  n8 = LinearInterpolate(
   B=TextureSample(UVs=TextureCoordinate() {uTiling=8, vTiling=8}) {texture=/Game/Effects/Fort_Effects/Textures/Masks/T_Vert_Gradient_Tilable}.G
   Alpha=@n3.DirectionalFade
  ) {constA=1}
  n9 = Multiply(A=ShadingPathSwitch(Default=DepthFade(Opacity=@n6, FadeDistance=@n7), Mobile=@n6) {desc=Migrated from FeatureLevelSwitch}, B=@n8)
  n10 = ScalarParameter() {defaultValue=100, parameterName=NearFadeDistance}
  n11 = StaticSwitchParameter(
   True=Multiply(A=@n9, B=MaterialFunctionCall(Distance=@n10) {materialFunction=&Near_Camera_Fade}.Result)
   False=@n9
  ) {parameterName=NearCameraFade}
  n12 = StaticBoolParameter() {defaultValue=true, parameterName="IsFallingSnowFX (B)"}
  LinearInterpolate(
   A=@n11
   B=Multiply(
    A=MaterialFunctionCall!(
     IsStaticSnow
     UseAlternateStaticSnowMask
     AlternateStaticSnowMask
     "UseRoad&&OcclusionMask"
     IsForGrassMeshToPushDown
     AdditionalMaskToRemoveSnow
     IsFallingSnowFX=@n12
     UseMinimapCoords
     IncreaseSnowUsingSlope
     UseSmoothBorder
     LerpHighM
     TemporalHeightLerpContrast
     IsLobbyIsland
     UseBR-IslandMask
     IsTree
     IsForRenderTarget
     UseSparkle
     UseSnowNormals
     UseAdditionalMaskToRemoveSnow
     IsTangentSpaceInput
     UseVertexNormalsForBasicSnow
     UseTangentSnowNormals
     Material
    ) {materialFunction=/Game/Athena/Apollo/Environments/Landscape/Materials/Biome/UniqueBiomeMaterialFunctions/MF_Apollo_Snow_Progression}.SimpleMask
    B=@n11
   )
   Alpha=ScalarParameter() {parameterName=UseSnowProgression}
  )
  n13 = VectorParameter() {parameterName=DarkColor, defaultValue=<rgba 0.0896 0.1113 0.2 0>}
  n14 = Desaturation(None=@n1.RGB, Fraction=Constant() {r=0.4})
  n15 = ParticleSubUV() {samplerType=SAMPLERTYPE_Normal, texture=/Game/Effects/Fort_Effects/Textures/Smoke/T_SmokeFlip_03_8x8_N}
  n16 = Constant3Vector() {constant=<rgba -1 -1 1 0>}
  n17 = CollectionParameter() {
   collection=&FortniteMaterialParameters
   parameterName=SunAndMoonModelDirectionalVector
  }
  n18 = DotProduct(
   A=Transform(None=Multiply(A=@n15.RGB, B=@n16))
   B=LinearInterpolate(
    A=ComponentMask(None=@n17) {b=true, g=true, r=true}
    B=VectorParameter() {parameterName=FakeLightAngle, defaultValue=<rgba 0 0 -0.8 1>}.RGB
    Alpha=ScalarParameter() {parameterName=UseActorPositionAsLightSource}
   )
  )
  n19 = Constant() {r=0.5}
  n20 = StaticSwitchParameter(True=ShadingPathSwitch(Default=Clamp(None=@n18), Mobile=@n19) {desc=Migrated from FeatureLevelSwitch}, False=@n19) {defaultValue=true, parameterName=Use Sun Vector DOT}
  n21 = Multiply(A=@n2.G, B=@n8)
  MaterialFunctionCall(Emissive=LinearInterpolate(A=Multiply(A=@n13.RGB, B=@n14), B=@n1.RGB, Alpha=Multiply(A=@n20, B=@n21))) {materialFunction=&MF_FX_EmissiveAdjustments}
  n22 = ParticlePositionWS()
  n23 = ShadingPathSwitch(Default=@n22, Deferred=@n22, Mobile=ObjectPositionWS()) {desc="Please fix particle position on mobile :("}
  n24 = WorldPosition()
  n25 = Divide(A=Distance(A=@n23, B=CameraPositionWS()), B=@n10)
  StaticSwitchParameter(True=LinearInterpolate(A=Subtract(A=@n23, B=@n24.XYZ), Alpha=Clamp(None=@n25)) {constB=0}, False=Constant()) {defaultValue=true, parameterName=NearCameraShrink}
 ]
}
