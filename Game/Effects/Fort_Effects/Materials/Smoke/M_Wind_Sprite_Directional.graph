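# Unlit translucent sprite material for directional wind or smoke trails, usable in Cascade and Niagara particles.
# A panned smoke trail blends with panned noise, and dynamic parameters drive pan timing, Y tiling and falloff shape.
# Fake lighting dots the trail normal map with the sun/moon vector to blend dark tint into particle color.
# A warped sphere mask shapes opacity, with near-camera fade, plus depth fade and optional fresnel skipped on mobile.
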Material /Game/Effects/Fort_Effects/Materials/Smoke/M_Wind_Sprite_Directional
graph {
 properties {
  bUsedWithNiagaraSprites=true
  bUsedWithParticleSprites=true
  blendMode=BLEND_Translucent
  pixelDepthOffsetMode=PDOM_Legacy
  shadingModel=MSM_Unlit
  usedShadingModels=Unlit
 }
 nodes [
  n1 = VectorParameter() {parameterName=DarkColor, defaultValue=<rgba 0.0729 0.0801 0.1094 0>}
  n2 = ParticleColor()
  n3 = Desaturation(None=@n2.RGB, Fraction=Constant() {r=0.5})
  n4 = TextureCoordinate()
  n5 = DynamicParameter() {defaultValue=<rgba 1 0.5 1 1>, paramNames [PanAdd,PanSpeed,YTile,LengthPower]}
  n6 = AppendVector(A=Constant() {r=1}, B=@n5.YTile)
  n7 = Multiply(A=Add(A=Time(), B=@n5.PanAdd), B=@n5.PanSpeed)
  n8 = Panner(
   Coordinate=MaterialFunctionCall(XY=Multiply(A=@n4, B=@n6)) {materialFunction=&Swizzle}.YX
   Time=@n7
   Speed=VectorParameter() {parameterName=Base_Panner_Speed, defaultValue=<rgba 0.8 0 0 1>}.RGB
  ) {speedX=0.8}
  n9 = TextureSample(UVs=@n8) {samplerType=SAMPLERTYPE_Normal, texture=/Game/Effects/Fort_Effects/Textures/Smoke/T_Smoke_Trail_Soft_N}
  n10 = Constant3Vector() {constant=<rgba -1 -1 1 0>}
  n11 = CollectionParameter() {
   collection=&FortniteMaterialParameters
   parameterName=SunAndMoonModelDirectionalVector
  }
  n12 = DotProduct(A=Transform(None=Multiply(A=@n9.RGB, B=@n10)), B=ComponentMask(None=@n11) {b=true, g=true, r=true})
  n13 = ShadingPathSwitch(Default=Clamp(None=Abs(None=@n12)), Mobile=Constant() {r=0.5}) {desc=Migrated from FeatureLevelSwitch}
  n14 = Multiply(
   A=Multiply(A=@n6, B=TextureCoordinate() {uTiling=0.7, vTiling=0.5})
   B=ComponentMask(None=VectorParameter() {parameterName=Noise_Tile, defaultValue=<rgba 1 1 0 1>}.RGB) {g=true, r=true}
  )
  n15 = VectorParameter() {parameterName=Noise_Panner_Speed, defaultValue=<rgba 0 1 0 1>}
  n16 = LinearInterpolate(
   A=TextureSample(UVs=@n8) {samplerType=SAMPLERTYPE_Grayscale, texture=&T_Smoke_Trail_Soft}.G
   B=TextureSampleParameter2D(UVs=Panner(Coordinate=@n14, Time=@n7, Speed=@n15.RGB) {speedY=1}) {parameterName=Noise, texture=&T_Smooth_tile}.G
  )
  MaterialFunctionCall(Emissive=LinearInterpolate(A=Multiply(A=@n1.RGB, B=@n3), B=@n2.RGB, Alpha=Multiply(A=@n13, B=@n16))) {materialFunction=&MF_FX_EmissiveAdjustments}
  n17 = MaterialFunctionCall(Distance=ScalarParameter() {defaultValue=200, parameterName=NearCamFade}) {materialFunction=&Near_Camera_Fade}
  n18 = TextureCoordinate()
  n19 = AppendVector(A=ComponentMask(None=@n18) {r=true}, B=Power(Base=ComponentMask(None=@n18) {g=true}, Exp=@n5.LengthPower))
  n20 = Constant() {r=0.5}
  n21 = Multiply(A=LinearInterpolate(Alpha=@n16) {constA=0.15}, B=Power(Base=SphereMask(A=@n19, B=@n20, Radius=@n20) {hardnessPercent=25}))
  n22 = DepthFade(Opacity=@n21, FadeDistance=ScalarParameter() {defaultValue=25, parameterName=DepthFadeDistance})
  n23 = LinearInterpolate(
   A=Constant() {r=1}
   B=MaterialFunctionCall(
    Brightness=ScalarParameter() {defaultValue=1, parameterName=Fresnel_Brightness}
    Exponent=ScalarParameter() {defaultValue=1, parameterName=Fresnel_Exp}
   ) {materialFunction=&MF_Fresnel_FX}.Fresnel
   Alpha=ScalarParameter() {parameterName=Use_Fresnel}
  )
  n24 = Multiply(A=@n2.A, B=ShadingPathSwitch(Default=Multiply(A=@n22, B=@n23), Mobile=@n21) {desc=Migrated from FeatureLevelSwitch})
  Power(
   Base=Multiply(A=Multiply(A=@n17.Result, B=@n24), B=ScalarParameter() {defaultValue=1, parameterName=Opacity_Mult})
   Exp=ScalarParameter() {defaultValue=1, parameterName=Opacity_Power}
  )
 ]
}
