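# Lit translucent particle sprite material for curly smoke wisps that erode away as particle alpha drops.
# An inky smoke flipbook, warped by a panning normal map, feeds emissive and a smoothstep dissolve driven by dynamic-parameter softness.
# A randomly offset tiling smoke texture with depth fade and optional near-camera fade modulates opacity.
# Switches add panning rainbow color or softer emissive edges; low quality skips warp and depth fade.
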Material /Game/Effects/Fort_Effects/Materials/Smoke/m_Smoke_CurlyWispDissolve
graph {
 properties {
  bUsedWithNiagaraSprites=true
  bUsedWithParticleSprites=true
  blendMode=BLEND_Translucent
  pixelDepthOffsetMode=PDOM_Legacy
  usedShadingModels=Default Lit
 }
 nodes [
  n1 = ComponentMask(None=TextureCoordinate() {uTiling=0, vTiling=0}) {g=true}
  n2 = Multiply(A=Time(), B=DynamicParameter() {paramNames [SmoothstepSoftness, Random, WarpStrength, Panning Speed]}.Panning Speed)
  n3 = Multiply(
   A=TextureSample(UVs=Panner(Coordinate=@n1, Time=@n2) {speedX=1.6667}) {texture=/Game/Global/Textures/Default/Masks_and_Gradients/T_Grad_Rainbow}.RGB
   B=ParticleColor().RGB
  )
  n4 = Clamp(None=@n3)
  n5 = MaterialFunctionCall(LDR Input=@n4, HDR Tint=@n3) {materialFunction=&DeriveHDRfromLDR}
  n6 = ParticleColor()
  n7 = TextureCoordinate()
  n8 = DynamicParameter() {paramNames [SmoothstepSoftness, Random, WarpStrength, Panning Speed]}
  n9 = Constant2Vector() {g=8, r=3}
  n10 = Add(A=TextureCoordinate(), B=Frac(None=Multiply(A=@n8.Random, B=@n9)))
  n11 = Panner(Coordinate=Multiply(A=ScalarParameter() {defaultValue=1.5, parameterName=WarpTiling}, B=@n10)) {speedX=0.1, speedY=0.025}
  n12 = Multiply(
   A=ComponentMask(None=TextureSample(UVs=@n11) {samplerType=SAMPLERTYPE_Normal, texture=&T_Weap_Distort_N_02}.RGB) {g=true, r=true}
   B=DynamicParameter() {paramNames [SmoothstepSoftness, Random, WarpStrength, Panning Speed]}.WarpStrength
  )
  n13 = TextureSample(UVs=QualitySwitch(Default=Add(A=@n7, B=@n12), Low=@n7)) {texture=/Game/Effects/Fort_Effects/Textures/T_InkySmoke_Flipbook}
  n14 = Multiply(A=@n13.G, B=@n13.G)
  n15 = Multiply(A=@n14, B=@n14)
  MaterialFunctionCall(Emissive=Multiply(
    A=StaticSwitchParameter(True=@n5.RemappedRange, False=@n6.RGB) {parameterName=Use Rainbow}
    B=StaticSwitchParameter(True=@n13.G, False=@n15) {parameterName=SofterEdge}
   )) {materialFunction=&MF_FX_EmissiveAdjustments}
  n16 = OneMinus(None=ParticleColor().A)
  n17 = Add(A=@n16, B=DynamicParameter() {paramNames [SmoothstepSoftness, Random, WarpStrength, Panning Speed]}.SmoothstepSoftness)
  n18 = TextureSample(UVs=@n10) {texture=&T_Inky_Smoke_Tile}
  n19 = QualitySwitch(Default=DepthFade(Opacity=Power(Base=@n18.G) {constExponent=0.5}) {fadeDistanceDefault=80}, Low=Constant() {r=1})
  n20 = Multiply(
   A=MaterialFunctionCall(Distance=ScalarParameter() {defaultValue=256, parameterName=Near Camera Fade Distance}) {materialFunction=&Near_Camera_Fade}.Result
   B=@n19
  )
  Multiply(
   A=Multiply(
    A=MaterialFunctionCall(Alpha=@n13.G, Min=@n16, Max=@n17) {materialFunction=&SmoothStep}.Result
    B=StaticSwitchParameter(True=@n20, False=@n19) {parameterName=Use Near Camera Fade}
   )
   B=ScalarParameter() {defaultValue=1.2, parameterName=OverallOpacity}
  )
  StaticSwitchParameter(True=@n4, False=VectorParameter() {parameterName=BaseColor, defaultValue=<rgba 0.0949 0.0963 0.1042 0>}.RGB) {parameterName=Use Rainbow}
 ]
}
